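# Looping indoor daytime ambience cue for the Bug Biome, reusing the exterior loop recordings at reduced volume and pitch under an indoor ambient sound class.
# A quality-level node sends its first two slots to the quad wave and its last two to the stereo wave.
# Each branch nests two indefinitely looping envelopers of different periods, one modulating pitch and one volume, so the bed drifts slowly.
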SoundCue /Hermes/Audio/BugBiome/SC_Ambience_BugBiome_Day_Loop_Int
root=QualityLevel_2
nodes [
 {
  id=Enveloper_0
  children [WavePlayer_2]
  properties {
   bLoop=true
   bLoopIndefinitely=true
   loopEnd=30
   pitchCurve {editorCurveData {keys [
      {arriveTangent=-0.0013, interpMode=RCIM_Cubic, leaveTangent=-0.0013, time=0}
      <at 15 0.98 RCIM_Cubic>
      {arriveTangent=0.0013, interpMode=RCIM_Cubic, leaveTangent=0.0013, time=30}
     ]}}
   volumeCurve {editorCurveData {keys [<at 0 1>]}}
  }
 }
 {
  id=Enveloper_1
  children [WavePlayer_1]
  properties {
   bLoop=true
   bLoopIndefinitely=true
   loopEnd=30
   pitchCurve {editorCurveData {keys [
      {arriveTangent=-0.0013, interpMode=RCIM_Cubic, leaveTangent=-0.0013, time=0}
      <at 15 0.98 RCIM_Cubic>
      {arriveTangent=0.0013, interpMode=RCIM_Cubic, leaveTangent=0.0013, time=30}
     ]}}
   volumeCurve {editorCurveData {keys [<at 0 1>]}}
  }
 }
 {
  id=Enveloper_2
  children [Enveloper_1]
  properties {
   bLoop=true
   bLoopIndefinitely=true
   loopEnd=20
   pitchCurve {editorCurveData {keys [{time=0}]}}
   volumeCurve {editorCurveData {keys [
      {arriveTangent=-0.01, interpMode=RCIM_Cubic, leaveTangent=-0.01, time=0, value=1}
      <at 10 0.9 RCIM_Cubic>
      {arriveTangent=0.01, interpMode=RCIM_Cubic, leaveTangent=0.01, time=20, value=1}
     ]}}
  }
 }
 {
  id=Enveloper_3
  children [Enveloper_0]
  properties {
   bLoop=true
   bLoopIndefinitely=true
   loopEnd=20
   pitchCurve {editorCurveData {keys [{time=0}]}}
   volumeCurve {editorCurveData {keys [
      {arriveTangent=-0.01, interpMode=RCIM_Cubic, leaveTangent=-0.01, time=0, value=1}
      <at 10 0.9 RCIM_Cubic>
      {arriveTangent=0.01, interpMode=RCIM_Cubic, leaveTangent=0.01, time=20, value=1}
     ]}}
  }
 }
 {id=QualityLevel_2, children [Enveloper_3,Enveloper_3,Enveloper_2,Enveloper_2]}
 {id=WavePlayer_1, properties {bLooping=true, soundWaveAssetPtr=&/Ambience_BugBiome_Day_01_Loop_Ext_Stereo}}
 {id=WavePlayer_2, properties {bLooping=true, soundWaveAssetPtr=&/Ambience_BugBiome_Day_01_Loop_Ext_Quad}}
]
properties {
 attenuationSettings=/AmbientAudioBase/Attenuations/Fort_Ambient_Athena_Att
 maxDistance=2097152
 pitchMultiplier=0.9
 soundClassObject=&Fort_Ambient_Indoors_Athena
 volumeMultiplier=0.2
}
