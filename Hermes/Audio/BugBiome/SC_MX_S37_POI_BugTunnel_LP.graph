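# Sound cue that plays the looping music track for the Bug Tunnel POI in the Hermes bug biome.
# A fixed delay holds playback, then an Enveloper fades the looping wave in from silence over a few seconds.
# It routes to the music sound class at reduced volume with no-verb musical attenuation.
# A concurrency override allows one instance and prevents new ones from starting while it plays.
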SoundCue /Hermes/Audio/BugBiome/SC_MX_S37_POI_BugTunnel_LP
root=Delay_0
nodes [
 {id=Delay_0, children [Enveloper_0], properties {delayMax=2.5, delayMin=2.5}}
 {
  id=Enveloper_0
  children [WavePlayer_0]
  properties {pitchCurve {editorCurveData {keys [{time=0}]}}, volumeCurve {editorCurveData {keys [<at 0 0>,<at 2.5 1>]}}}
 }
 {id=WavePlayer_0, properties {bLooping=true, soundWaveAssetPtr=&/MX_S37_POI_BugTunnel_LP}}
]
properties {
 attenuationSettings=/Game/Sounds/Attenuation/Fort_Ambient_Athena_Musical_Oneshot_NoVerb_Att
 bOverrideConcurrency=true
 maxDistance=2097152
 soundClassObject=/Game/Sounds/FortSoundClassesAndModes/Fort_Music
 volumeMultiplier=0.5
 concurrencyOverrides {maxCount=1, resolutionRule=PreventNew, voiceStealReleaseTime=2}
}
